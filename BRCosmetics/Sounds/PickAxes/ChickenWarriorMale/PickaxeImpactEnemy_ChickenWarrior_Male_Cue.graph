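# Sound cue for the Chicken Warrior Male pickaxe hitting an enemy, with melee concurrency and pickaxe impact attenuation.
# A quality-level switch picks among branches that use six, three or two variations each of impact and vox sounds, with two levels sharing the three-variation branch.
# Each branch usually plays an impact alone, but a weighted random occasionally mixes in a quieter, slightly delayed vox.
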SoundCue /BRCosmetics/Sounds/PickAxes/ChickenWarriorMale/PickaxeImpactEnemy_ChickenWarrior_Male_Cue
root=QualityLevel_0
nodes [
 {id=Delay_0, children [Modulator_0], properties {delayMax=0.05, delayMin=0.05}}
 {id=Delay_3, children [Modulator_5], properties {delayMax=0.05, delayMin=0.05}}
 {id=Delay_4, children [Modulator_4], properties {delayMax=0.05, delayMin=0.05}}
 {id=Mixer_0, children [Modulator_1,Delay_0], properties {inputVolume [1,0.4]}}
 {id=Mixer_1, children [Modulator_2,Delay_4], properties {inputVolume [1,0.4]}}
 {id=Mixer_2, children [Modulator_3,Delay_3], properties {inputVolume [1,0.4]}}
 {id=Modulator_0, children [Random_0]}
 {id=Modulator_1, children [Random_1]}
 {id=Modulator_2, children [Random_8]}
 {id=Modulator_3, children [Random_9]}
 {id=Modulator_4, children [Random_3]}
 {id=Modulator_5, children [Random_11]}
 {id=QualityLevel_0, children [Random_5,Random_7,Random_10,Random_7]}
 {id=Random_0, children [WavePlayer_0,WavePlayer_1,WavePlayer_2,WavePlayer_3,WavePlayer_4,WavePlayer_5], properties {weights [1,1,1,1,1,1]}}
 {
  id=Random_1
  children [WavePlayer_12, WavePlayer_13, WavePlayer_14, WavePlayer_15, WavePlayer_26, WavePlayer_27]
  properties {weights [1,1,1,1,1,1]}
 }
 {id=Random_10, children [Modulator_3,Mixer_2], properties {bRandomizeWithoutReplacement=false, weights [1,0.2]}}
 {id=Random_11, children [WavePlayer_43, WavePlayer_44], properties {weights [1,1]}}
 {id=Random_3, children [WavePlayer_28, WavePlayer_29, WavePlayer_30], properties {weights [1,1,1]}}
 {id=Random_5, children [Modulator_1,Mixer_0], properties {bRandomizeWithoutReplacement=false, weights [1,0.25]}}
 {id=Random_7, children [Modulator_2,Mixer_1], properties {bRandomizeWithoutReplacement=false, weights [1,0.2]}}
 {id=Random_8, children [WavePlayer_38, WavePlayer_39, WavePlayer_40], properties {weights [1,1,1]}}
 {id=Random_9, children [WavePlayer_41, WavePlayer_42], properties {weights [1,1]}}
 {id=WavePlayer_0, properties {soundWaveAssetPtr=&/PickAxe_EnemyImpact_ChickenWarrior_Vox_01}}
 {id=WavePlayer_1, properties {soundWaveAssetPtr=&/PickAxe_EnemyImpact_ChickenWarrior_Vox_02}}
 {id=WavePlayer_12, properties {soundWaveAssetPtr=&/PickAxe_Chicken_Warrior_Male_Impact_01}}
 {id=WavePlayer_13, properties {soundWaveAssetPtr=&/PickAxe_Chicken_Warrior_Male_Impact_03}}
 {id=WavePlayer_14, properties {soundWaveAssetPtr=&/PickAxe_Chicken_Warrior_Male_Impact_05}}
 {id=WavePlayer_15, properties {soundWaveAssetPtr=&/PickAxe_Chicken_Warrior_Male_Impact_02}}
 {id=WavePlayer_2, properties {soundWaveAssetPtr=&/PickAxe_EnemyImpact_ChickenWarrior_Vox_03}}
 {id=WavePlayer_26, properties {soundWaveAssetPtr=&/PickAxe_Chicken_Warrior_Male_Impact_04}}
 {id=WavePlayer_27, properties {soundWaveAssetPtr=&/PickAxe_Chicken_Warrior_Male_Impact_06}}
 {id=WavePlayer_28, properties {soundWaveAssetPtr=&PickAxe_EnemyImpact_ChickenWarrior_Vox_01}}
 {id=WavePlayer_29, properties {soundWaveAssetPtr=&PickAxe_EnemyImpact_ChickenWarrior_Vox_02}}
 {id=WavePlayer_3, properties {soundWaveAssetPtr=&/PickAxe_EnemyImpact_ChickenWarrior_Vox_04}}
 {id=WavePlayer_30, properties {soundWaveAssetPtr=&PickAxe_EnemyImpact_ChickenWarrior_Vox_03}}
 {id=WavePlayer_38, properties {soundWaveAssetPtr=&PickAxe_Chicken_Warrior_Male_Impact_01}}
 {id=WavePlayer_39, properties {soundWaveAssetPtr=&PickAxe_Chicken_Warrior_Male_Impact_03}}
 {id=WavePlayer_4, properties {soundWaveAssetPtr=&/PickAxe_EnemyImpact_ChickenWarrior_Vox_05}}
 {id=WavePlayer_40, properties {soundWaveAssetPtr=&PickAxe_Chicken_Warrior_Male_Impact_02}}
 {id=WavePlayer_41, properties {soundWaveAssetPtr=&PickAxe_Chicken_Warrior_Male_Impact_01}}
 {id=WavePlayer_42, properties {soundWaveAssetPtr=&PickAxe_Chicken_Warrior_Male_Impact_02}}
 {id=WavePlayer_43, properties {soundWaveAssetPtr=&PickAxe_EnemyImpact_ChickenWarrior_Vox_01}}
 {id=WavePlayer_44, properties {soundWaveAssetPtr=&PickAxe_EnemyImpact_ChickenWarrior_Vox_02}}
 {id=WavePlayer_5, properties {soundWaveAssetPtr=&/PickAxe_EnemyImpact_ChickenWarrior_Vox_06}}
]
properties {
 attenuationSettings=/Game/Sounds/Attenuation/Fort_Pickaxe_Impact_Att
 audioPropertiesSheet=/BRCosmetics/Sounds/Settings/PropertySheets/Pickaxe/APS_Cosmetic_Pickaxe_Impact
 bPrimeOnLoad=true
 maxDistance=4300
 soundClassObject=&Fort_WeaponMelee_PSM
 volumeMultiplier=0.5
 concurrencySet [&Player_Melee_Weapon_Conc]
}
